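# Sound cue played when interacting with the katana container spawner, routed through the loot-collect sound class with its own attenuation.
# A mixer layers a main interact wave, a quieter layer wave on a short fixed delay, and a pitched-up, randomly chosen katana equip variant.
# A root modulator adds slight random upward pitch variation to the combined mix.
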SoundCue /KatanaGameplay/Items/Environmental/KatanaContainer/Sound/Katana_Spawner_Interact_Cue
root=Modulator_0
nodes [
 {id=Delay_1, children [WavePlayer_12], properties {delayMax=0.1, delayMin=0.1}}
 {id=Mixer_0, children [WavePlayer_0,Delay_1,Modulator_1], properties {inputVolume [1.1,0.6,0.6]}}
 {id=Modulator_0, children [Mixer_0], properties {pitchMax=1.1, pitchMin=1}}
 {id=Modulator_1, children [Random_2], properties {pitchMax=1.2, pitchMin=1.2}}
 {id=Random_2, children [WavePlayer_9, WavePlayer_10, WavePlayer_11], properties {weights [1,1,1]}}
 {id=WavePlayer_0, properties {soundWaveAssetPtr=&/Katana_Spawner_Interact_01}}
 {id=WavePlayer_10, properties {soundWaveAssetPtr=/KatanaGameplay/Items/Katana/Sounds/Katana_EquipB_02}}
 {id=WavePlayer_11, properties {soundWaveAssetPtr=/KatanaGameplay/Items/Katana/Sounds/Katana_EquipB_03}}
 {id=WavePlayer_12, properties {soundWaveAssetPtr=&/Katana_Spawner_Interact_Layer_01}}
 {id=WavePlayer_9, properties {soundWaveAssetPtr=/KatanaGameplay/Items/Katana/Sounds/Katana_EquipB_01}}
]
properties {
 attenuationSettings=&/Katana_Spawner_Interact_Att
 maxDistance=2500
 soundClassObject=&Fort_Loot_Collect
 volumeMultiplier=0.9
 concurrencySet [null]
}
