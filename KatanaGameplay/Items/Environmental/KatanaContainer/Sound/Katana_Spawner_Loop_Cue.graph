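# Sound cue for the looping ambient sound of the Katana spawner container, classed as ambient loot audio.
# A single looping wave player feeds the output directly, with no randomization or mixing.
# Spatial falloff comes from the Katana chest attenuation asset, and the shared loot ambient concurrency set limits simultaneous instances.
# The cue also stores attenuation and concurrency override values, but no override switch is turned on.
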SoundCue /KatanaGameplay/Items/Environmental/KatanaContainer/Sound/Katana_Spawner_Loop_Cue
root=WavePlayer_0
nodes [{
  id=WavePlayer_0
  properties {bLooping=true, soundWaveAssetPtr=&/Katana_Spawner_Loop_02}
 }]
properties {
 attenuationSettings=&/Katana_Chests_Att
 maxDistance=1700
 priority=50
 soundClassObject=/Game/Sounds/FortSoundClassesAndModes/Fort_Loot_Ambient_Loop
 volumeMultiplier=0.8
 attenuationOverrides {
  bEnableFocusInterpolation=true
  bEnableReverbSend=false
  falloffDistance=1000
  focusAttackInterpSpeed=20
  focusAzimuth=20
  focusReleaseInterpSpeed=10
  focusVolumeAttenuation=1.75
  lPFFrequencyAtMax=10000
  lPFRadiusMax=1000
  lPFRadiusMin=500
  nonFocusAzimuth=40
  nonFocusVolumeAttenuation=0.8
  nonSpatializedRadiusStart=10
  occlusionInterpolationTime=0.3
  occlusionLowPassFilterFrequency=3000
  occlusionVolumeAttenuation=0.75
  reverbWetLevelMax=0.2
  reverbWetLevelMin=0.2
  customAttenuationCurve {editorCurveData {keys [<at 0 1 RCIM_Cubic>, {arriveTangent=-1, interpMode=RCIM_Cubic, leaveTangent=-1, time=0.5, value=0.1}, <at 1 0 RCIM_Cubic>]}}
 }
 concurrencyOverrides {bVolumeScaleCanRelease=true, maxCount=5, resolutionRule=StopQuietest, volumeScale=0.5}
 concurrencySet [/Game/Sounds/Concurrency_Rules/Loot_Ambient_Conc]
}
